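# Additive, unlit, two-sided Niagara particle material for the 3DHB Ghost FX preset, drawing a color-fringed ghost of a mask texture.
# A custom HLSL loop averages dithered alpha samples along a direction, tinted across three colors by an injected Lerp3Color helper, then feeds emissive adjustments.
# Mask alpha, flippable via IsFrom3D, is multiplied by hologram scanline and glitch, particle alpha and inverse Fresnel.
# A per-particle Random offsets hologram time.
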Material /Game/Tools/3DHB/FX/PresetFX/Ghost/M_3DHB_FX_Ghost
graph {
 properties {
  bUsedWithNiagaraMeshParticles=true
  bUsedWithNiagaraSprites=true
  blendMode=BLEND_Additive
  pixelDepthOffsetMode=PDOM_Legacy
  shadingModel=MSM_Unlit
  twoSided=true
  usedShadingModels=Unlit
 }
 nodes [
  MaterialExpressionCustom_4 = Custom() {
   code="return 1;\r\n}\r\n\r\nhalf3 Lerp3Color (half Alpha,half3 R,half3 G,half3 B){\r\n  \r\nreturn lerp(   lerp(R,G,saturate(Alpha*2))   ,B,saturate(Alpha*2-1));\r\n"
   description=Lerp3ColorFunction
   outputType=CMOT_Float1
   inputs []
  }
  MaterialExpressionTextureObjectParameter_0 = TextureObjectParameter() {parameterName=Ghost_MaskTexture, texture=/Game/Tools/3DHB/DefaultAssets/T_3DHB_DefaultTexture}
  MaterialExpressionConstant_4 = Constant() {r=4}
  MaterialExpressionScalarParameter_1 = ScalarParameter() {defaultValue=0.4, parameterName=Ghost_Distance}
  n1 = TextureCoordinate()
  n2 = Add(A=Multiply(A=Add(A=@n1) {constB=-0.5}, B=Constant() {r=2})) {constB=0.5}
  MaterialExpressionAdd_4 = Add(
   A=@n2
   B=AppendVector(A=ScalarParameter() {defaultValue=-0.14, parameterName=Ghost_OffsetX}, B=ScalarParameter() {parameterName=Ghost_OffsetY})
  )
  MaterialExpressionVectorParameter_0 = VectorParameter() {parameterName=Ghost_Direction, defaultValue=<rgba 1 0 0 1>}
  MaterialExpressionVectorParameter_6 = VectorParameter() {parameterName=Ghost_Color_R, defaultValue=<rgba 0 0.0192 1 1>}
  MaterialExpressionVectorParameter_7 = VectorParameter() {parameterName=Ghost_Color_G, defaultValue=<rgba 0 1 0.4557 1>}
  MaterialExpressionVectorParameter_8 = VectorParameter() {parameterName=Ghost_Color_B, defaultValue=<rgba 0 0.0786 1 1>}
  MaterialExpressionScreenPosition_0 = ScreenPosition()
  n3 = Custom!(ScreenPos=@MaterialExpressionScreenPosition_0.PixelPosition) {
   code="float random = InterleavedGradientNoise(ScreenPos.xy, float(View.StateFrameIndexMod8));\r\nreturn random;"
   outputType=CMOT_Float1
   inputs [{inputName=ScreenPos, input {expression=&self:MaterialExpressionScreenPosition_0, outputIndex=1}}]
  }
  MaterialExpressionSubtract_0 = Subtract(A=@n3) {constB=0.5}
  MaterialExpressionScalarParameter_7 = ScalarParameter() {parameterName=IsFrom3D}
  n4 = Custom!(
   Function=@MaterialExpressionCustom_4
   T2d=@MaterialExpressionTextureObjectParameter_0
   Amount=@MaterialExpressionConstant_4
   DistanceSum=@MaterialExpressionScalarParameter_1
   UV=@MaterialExpressionAdd_4
   Dir=@MaterialExpressionVectorParameter_0.RGB
   R=@MaterialExpressionVectorParameter_6.RGB
   G=@MaterialExpressionVectorParameter_7.RGB
   B=@MaterialExpressionVectorParameter_8.RGB
   Dither=@MaterialExpressionSubtract_0
   From3D=@MaterialExpressionScalarParameter_7
  ) {
   code="int IntAmount = floor(Amount);\r\nhalf PDistance = DistanceSum/IntAmount;\r\nhalf3 SumWeight = 0;\r\nhalf3 SumColor = 0;\r\nfor(int i = 0 ; i < IntAmount ; i++)\r\n    {\r\n        half3 ColorM = Lerp3Color(i/(half(IntAmount-1)),R,G,B);\r\n        half2 Pos = Dir*(PDistance*(i+Dither)) + UV;\r\n        half3 SampledColor = (Texture2DSample(T2d,GetMaterialSharedSampler(Material.Texture2D_0Sampler,View.MaterialTextureBilinearClampedSampler), Pos)).w;\r\n        SampledColor = lerp(SampledColor,1-SampledColor,From3D);\r\n        SumColor += SampledColor*ColorM;\r\n        SumWeight += ColorM;\r\n}\r\n\r\nreturn SumColor/SumWeight;"
   inputs [
    {inputName=Function, input {expression=&self:MaterialExpressionCustom_4}}
    {inputName=T2d, input {expression=&self:MaterialExpressionTextureObjectParameter_0}}
    {inputName=Amount, input {expression=&self:MaterialExpressionConstant_4}}
    {inputName=DistanceSum, input {expression=&self:MaterialExpressionScalarParameter_1}}
    {inputName=UV, input {expression=&self:MaterialExpressionAdd_4}}
    {
     inputName=Dir
     input {expression=&self:MaterialExpressionVectorParameter_0, mask=1, maskB=1, maskG=1, maskR=1}
    }
    {
     inputName=R
     input {expression=&self:MaterialExpressionVectorParameter_6, mask=1, maskB=1, maskG=1, maskR=1}
    }
    {
     inputName=G
     input {expression=&self:MaterialExpressionVectorParameter_7, mask=1, maskB=1, maskG=1, maskR=1}
    }
    {
     inputName=B
     input {expression=&self:MaterialExpressionVectorParameter_8, mask=1, maskB=1, maskG=1, maskR=1}
    }
    {inputName=Dither, input {expression=&self:MaterialExpressionSubtract_0}}
    {inputName=From3D, input {expression=&self:MaterialExpressionScalarParameter_7}}
   ]
  }
  MaterialFunctionCall(Emissive=@n4) {materialFunction=&MF_FX_EmissiveAdjustments}
  n5 = MaterialFunctionCall!(
   Time=Add(A=Time() {bOverride_Period=true, period=6000}, B=DynamicParameter() {paramNames [Random,Param2,Param3,Param4]}.Random)
   MeshScaleIn
   UVForScanline
   CustomFlicker
   UseCustomFlicker
   UVForStroke
   UVForGlitch
   BaseUVIn
   EmissiveIn
   OpacityIn
   WPOIn
  ) {materialFunction=/Game/Tools/3DHB/Shader/Base/MF_3DHB_HologramFX}
  n6 = TextureSample(UVs=@n2, Tex=@MaterialExpressionTextureObjectParameter_0)
  n7 = OneMinus(None=@n6.A)
  n8 = Multiply(
   A=@n5.UseForFX_ScanlineOut
   B=Multiply(
    A=LinearInterpolate(A=@n7, B=@n6.A, Alpha=@MaterialExpressionScalarParameter_7)
    B=ScalarParameter() {defaultValue=0.2, parameterName=Ghost_OpacityScale}
   )
  )
  n9 = ScalarParameter() {defaultValue=10, parameterName=FresnelExp}
  Multiply(
   A=Multiply(A=Multiply(A=@n8, B=@n5.UseForFX_GlitchToOpacityOut), B=ParticleColor().A)
   B=OneMinus(None=Fresnel(ExponentIn=@n9) {baseReflectFraction=0})
  )
 ]
}
